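# Physics asset for the dress of the female medium Cyber Armor Alt cosmetic body, giving its skirt bones capsule bodies for physics motion.
# Eight three-segment skirt chains around front, back and sides each hang from a small MSTR anchor body constrained to the pelvis with free rotation.
# Swing and twist limits vary by chain to shape the motion, while larger capsules on pelvis, lower spine and thighs represent the body.
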PhysicsAsset /BRCosmetics/Characters/Player/Female/Medium/Bodies/F_MED_Cyber_Armor_Alt/Meshs/F_MED_Cyber_Armor_Alt_Dress_Physics
bodies {
 dyn_skirt_01_r [{shapeType=Capsule, length=9.3559, radius=2.6441, shapeName=dyn_skirt_01_r_capsule, rotation=<rot 0 0 89.9115>}]
 dyn_skirt_01_r_MSTR [{shapeType=Capsule, length=0.505, radius=0.505, shapeName=dyn_skirt_01_r_MSTR_capsule, rotation=<rot 0 -90 0>}]
 dyn_skirt_02_r [{shapeType=Capsule, length=10, radius=2, shapeName=dyn_skirt_02_r_capsule, rotation=<rot 0 0 90>}]
 dyn_skirt_03_r [{shapeType=Capsule, length=10, radius=2, shapeName=dyn_skirt_03_r_capsule, rotation=<rot 0 0 90>}]
 dyn_skirt_bk2_01_l [{
   shapeType=Capsule
   length=10.3915
   radius=1.6085
   shapeName=dyn_skirt_bk2_01_l_capsule
   center=<6.2129 0.0004 1.3059>
   rotation=<rot 0 0 89.9115>
  }]
 dyn_skirt_bk2_01_l_MSTR [{shapeType=Capsule, length=0.505, radius=0.505, shapeName=dyn_skirt_bk2_01_l_MSTR_capsule, rotation=<rot 0 -90 0>}]
 dyn_skirt_bk2_01_r [{shapeType=Capsule, length=9.3559, radius=2.6441, shapeName=dyn_skirt_bk2_01_r_capsule, center=<3.3149 0 0.0002>, rotation=<rot 0 0 89.9115>}]
 dyn_skirt_bk2_01_r_MSTR [{shapeType=Capsule, length=0.505, radius=0.505, shapeName=dyn_skirt_bk2_01_r_MSTR_capsule, rotation=<rot 0 -90 0>}]
 dyn_skirt_bk2_02_l [{shapeType=Capsule, length=10.823, radius=1.177, shapeName=dyn_skirt_bk2_02_l_capsule, center=<3.5227 0 -0.0003>, rotation=<rot 0 0 89.9999>}]
 dyn_skirt_bk2_02_r [{shapeType=Capsule, length=10, radius=2, shapeName=dyn_skirt_bk2_02_r_capsule, center=<3.7282 0 -0.0001>, rotation=<rot 0 0 89.9999>}]
 dyn_skirt_bk2_03_l [{shapeType=Capsule, length=10.823, radius=1.177, shapeName=dyn_skirt_bk2_03_l_capsule, center=<2.3005 0 -0.0003>, rotation=<rot 0 0 89.9999>}]
 dyn_skirt_bk2_03_r [{shapeType=Capsule, length=10, radius=2, shapeName=dyn_skirt_bk2_03_r_capsule, center=<2.5236 0 0>, rotation=<rot 0 0 89.9999>}]
 dyn_skirt_bk_01 [{
   shapeType=Capsule
   length=10.4903
   radius=1.5097
   shapeName=dyn_skirt_bk_01_capsule
   center=<5.4154 -0.0013 0.8273>
   rotation=<rot 0 0 89.9115>
  }]
 dyn_skirt_bk_01_MSTR [{shapeType=Capsule, length=0.505, radius=0.505, shapeName=dyn_skirt_bk_01_MSTR_capsule, rotation=<rot 0 -90 0>}]
 dyn_skirt_bk_01_l [{shapeType=Capsule, length=10.0284, radius=1.9716, shapeName=dyn_skirt_bk_01_l_capsule, center=<3.9485 0 -0.0004>, rotation=<rot 0 0 89.9115>}]
 dyn_skirt_bk_01_l_MSTR [{shapeType=Capsule, length=0.505, radius=0.505, shapeName=dyn_skirt_bk_01_l_MSTR_capsule, rotation=<rot 0 -90 0>}]
 dyn_skirt_bk_02 [{shapeType=Capsule, length=10, radius=2, shapeName=dyn_skirt_bk_02_capsule, center=<3.4344 0 -0.0001>, rotation=<rot 0 0 89.9999>}]
 dyn_skirt_bk_02_l [{shapeType=Capsule, length=10.8136, radius=1.1864, shapeName=dyn_skirt_bk_02_l_capsule, center=<3.9485 0 -0.0004>, rotation=<rot 0 0 89.9999>}]
 dyn_skirt_bk_03 [{shapeType=Capsule, length=10, radius=2, shapeName=dyn_skirt_bk_03_capsule, center=<3.4344 0 -0.0001>, rotation=<rot 0 0 89.9999>}]
 dyn_skirt_bk_03_l [{shapeType=Capsule, length=10.3096, radius=1.6904, shapeName=dyn_skirt_bk_03_l_capsule, center=<3.9485 0 -0.0004>, rotation=<rot 0 0 89.9999>}]
 dyn_skirt_fr_01 [{
   shapeType=Capsule
   length=26.525
   radius=2
   shapeName=dyn_skirt_fr_01_capsule
   center=<4.7295 -0.0013 0.8993>
   rotation=<rot 0.0054 0 89.9115>
  }]
 dyn_skirt_fr_01_MSTR [{shapeType=Capsule, length=0.505, radius=0.505, shapeName=dyn_skirt_fr_01_MSTR_capsule, rotation=<rot 0 -90 0>}]
 dyn_skirt_fr_01_l [{shapeType=Capsule, length=10, radius=2, shapeName=dyn_skirt_fr_01_l_capsule, rotation=<rot 0 0 90>}]
 dyn_skirt_fr_01_l_MSTR [{shapeType=Capsule, length=0.505, radius=0.505, shapeName=dyn_skirt_fr_01_l_MSTR_capsule, rotation=<rot 0 -90 0>}]
 dyn_skirt_fr_01_r [{shapeType=Capsule, length=10, radius=2, shapeName=dyn_skirt_fr_01_r_capsule, rotation=<rot 0 0 90>}]
 dyn_skirt_fr_01_r_MSTR [{shapeType=Capsule, length=0.505, radius=0.505, shapeName=dyn_skirt_fr_01_r_MSTR_capsule, rotation=<rot 0 -90 0>}]
 dyn_skirt_fr_02 [{shapeType=Capsule, length=26.525, radius=2, shapeName=dyn_skirt_fr_02_capsule, center=<4.4315 0 0>, rotation=<rot 0 0 89.9999>}]
 dyn_skirt_fr_02_l [{shapeType=Capsule, length=10, radius=2, shapeName=dyn_skirt_fr_02_l_capsule, rotation=<rot 0 0 90>}]
 dyn_skirt_fr_02_r [{shapeType=Capsule, length=10, radius=2, shapeName=dyn_skirt_fr_02_r_capsule, rotation=<rot 0 0 90>}]
 dyn_skirt_fr_03 [{shapeType=Capsule, length=26.525, radius=2, shapeName=dyn_skirt_fr_03_capsule, center=<3.05 0 0>, rotation=<rot 0 0 89.9999>}]
 dyn_skirt_fr_03_l [{shapeType=Capsule, length=10, radius=2, shapeName=dyn_skirt_fr_03_l_capsule, rotation=<rot 0 0 90>}]
 dyn_skirt_fr_03_r [{shapeType=Capsule, length=10, radius=2, shapeName=dyn_skirt_fr_03_r_capsule, rotation=<rot 0 0 90>}]
 pelvis [{shapeType=Capsule, length=7.4575, radius=8.476, shapeName=pelvis_capsule, center=<0 -1.5602 -0.0001>}]
 spine_01 [{shapeType=Capsule, length=11.6007, radius=6.3672, shapeName=spine_01_capsule}]
 spine_02 [{shapeType=Capsule, length=4.2003, radius=6.7972, shapeName=spine_02_capsule, center=<0 1.5317 0>}]
 thigh_l [
  {shapeType=Capsule, length=38.849, radius=7.0667, shapeName=thigh_l_capsule, center=<-21.2305 0 0>, rotation=<rot 90 2.5442 -177.5442>}
  {
   shapeType=Capsule
   length=10.9672
   radius=7.0667
   shapeName=thigh_l_capsule
   center=<-9.7047 -1.8581 -0.6213>
   rotation=<rot 85.4234 -179.9116 0>
  }
 ]
 thigh_r [{shapeType=Capsule, length=33.0077, radius=7.051, shapeName=thigh_r_capsule, center=<24.4294 0.0001 0>, rotation=<rot 90 -2.6 177.3116>}]
}
constraints [
 {
  bone1Name=spine_01
  bone2Name=pelvis
  swing1LimitDegrees=45
  swing1Motion=Free
  swing2LimitDegrees=45
  swing2Motion=Free
  twistLimitDegrees=45
  twistMotion=Free
 }
 {
  bone1Name=spine_02
  bone2Name=spine_01
  swing1LimitDegrees=45
  swing1Motion=Free
  swing2LimitDegrees=45
  swing2Motion=Free
  twistLimitDegrees=45
  twistMotion=Free
 }
 {
  bone1Name=dyn_skirt_fr_01_l_MSTR
  bone2Name=pelvis
  swing1LimitDegrees=45
  swing1Motion=Free
  swing2LimitDegrees=45
  swing2Motion=Free
  twistLimitDegrees=45
  twistMotion=Free
 }
 {
  bone1Name=dyn_skirt_fr_01_r_MSTR
  bone2Name=pelvis
  swing1LimitDegrees=45
  swing1Motion=Free
  swing2LimitDegrees=45
  swing2Motion=Free
  twistLimitDegrees=45
  twistMotion=Free
 }
 {
  bone1Name=dyn_skirt_fr_01_MSTR
  bone2Name=pelvis
  swing1LimitDegrees=45
  swing1Motion=Free
  swing2LimitDegrees=45
  swing2Motion=Free
  twistLimitDegrees=45
  twistMotion=Free
 }
 {
  bone1Name=dyn_skirt_bk_01_MSTR
  bone2Name=pelvis
  swing1LimitDegrees=45
  swing1Motion=Free
  swing2LimitDegrees=45
  swing2Motion=Free
  twistLimitDegrees=45
  twistMotion=Free
 }
 {
  bone1Name=dyn_skirt_bk2_01_l_MSTR
  bone2Name=pelvis
  swing1LimitDegrees=45
  swing1Motion=Free
  swing2LimitDegrees=45
  swing2Motion=Free
  twistLimitDegrees=45
  twistMotion=Free
 }
 {
  bone1Name=dyn_skirt_bk_01_l_MSTR
  bone2Name=pelvis
  swing1LimitDegrees=45
  swing1Motion=Free
  swing2LimitDegrees=45
  swing2Motion=Free
  twistLimitDegrees=45
  twistMotion=Free
 }
 {
  bone1Name=dyn_skirt_bk2_01_r_MSTR
  bone2Name=pelvis
  swing1LimitDegrees=45
  swing1Motion=Free
  swing2LimitDegrees=45
  swing2Motion=Free
  twistLimitDegrees=45
  twistMotion=Free
 }
 {
  bone1Name=dyn_skirt_01_r_MSTR
  bone2Name=pelvis
  swing1LimitDegrees=45
  swing1Motion=Free
  swing2LimitDegrees=45
  swing2Motion=Free
  twistLimitDegrees=45
  twistMotion=Free
 }
 {bone1Name=dyn_skirt_fr_01_l, bone2Name=dyn_skirt_fr_01_l_MSTR, swing1LimitDegrees=25, swing2LimitDegrees=45, twistLimitDegrees=5}
 {bone1Name=dyn_skirt_fr_02_l, bone2Name=dyn_skirt_fr_01_l, swing1LimitDegrees=25, swing2LimitDegrees=45, twistLimitDegrees=5}
 {bone1Name=dyn_skirt_fr_03_l, bone2Name=dyn_skirt_fr_02_l, swing1LimitDegrees=25, swing2LimitDegrees=45, twistLimitDegrees=5}
 {bone1Name=dyn_skirt_fr_01_r, bone2Name=dyn_skirt_fr_01_r_MSTR, swing1LimitDegrees=25, swing2LimitDegrees=45, twistLimitDegrees=5}
 {bone1Name=dyn_skirt_fr_02_r, bone2Name=dyn_skirt_fr_01_r, swing1LimitDegrees=25, swing2LimitDegrees=45, twistLimitDegrees=5}
 {bone1Name=dyn_skirt_fr_03_r, bone2Name=dyn_skirt_fr_02_r, swing1LimitDegrees=25, swing2LimitDegrees=45, twistLimitDegrees=5}
 {bone1Name=dyn_skirt_fr_01, bone2Name=dyn_skirt_fr_01_MSTR, swing1LimitDegrees=45, swing2LimitDegrees=45, twistLimitDegrees=45}
 {bone1Name=dyn_skirt_fr_02, bone2Name=dyn_skirt_fr_01, swing1LimitDegrees=45, swing2LimitDegrees=45, twistLimitDegrees=45}
 {bone1Name=dyn_skirt_fr_03, bone2Name=dyn_skirt_fr_02, swing1LimitDegrees=45, swing2LimitDegrees=45, twistLimitDegrees=45}
 {bone1Name=dyn_skirt_bk_01, bone2Name=dyn_skirt_bk_01_MSTR, swing1LimitDegrees=45, swing2LimitDegrees=45, twistLimitDegrees=45}
 {bone1Name=dyn_skirt_bk_02, bone2Name=dyn_skirt_bk_01, swing1LimitDegrees=45, swing2LimitDegrees=45, twistLimitDegrees=45}
 {bone1Name=dyn_skirt_bk_03, bone2Name=dyn_skirt_bk_02, swing1LimitDegrees=45, swing2LimitDegrees=45, twistLimitDegrees=45}
 {bone1Name=dyn_skirt_bk2_01_l, bone2Name=dyn_skirt_bk2_01_l_MSTR, swing1LimitDegrees=5, swing2LimitDegrees=45, twistLimitDegrees=5}
 {bone1Name=dyn_skirt_bk2_02_l, bone2Name=dyn_skirt_bk2_01_l, swing1LimitDegrees=5, swing2LimitDegrees=45, twistLimitDegrees=5}
 {bone1Name=dyn_skirt_bk2_03_l, bone2Name=dyn_skirt_bk2_02_l, swing1LimitDegrees=5, swing2LimitDegrees=45, twistLimitDegrees=5}
 {bone1Name=dyn_skirt_bk_01_l, bone2Name=dyn_skirt_bk_01_l_MSTR, swing1LimitDegrees=45, swing2LimitDegrees=45, twistLimitDegrees=45}
 {bone1Name=dyn_skirt_bk_02_l, bone2Name=dyn_skirt_bk_01_l, swing1LimitDegrees=5, swing2LimitDegrees=45, twistLimitDegrees=45}
 {bone1Name=dyn_skirt_bk_03_l, bone2Name=dyn_skirt_bk_02_l, swing1LimitDegrees=45, swing2LimitDegrees=45, twistLimitDegrees=45}
 {bone1Name=dyn_skirt_bk2_01_r, bone2Name=dyn_skirt_bk2_01_r_MSTR, swing1LimitDegrees=5, swing2LimitDegrees=45, twistLimitDegrees=45}
 {bone1Name=dyn_skirt_bk2_02_r, bone2Name=dyn_skirt_bk2_01_r, swing1LimitDegrees=5, swing2LimitDegrees=45, twistLimitDegrees=5}
 {bone1Name=dyn_skirt_bk2_03_r, bone2Name=dyn_skirt_bk2_02_r, swing1LimitDegrees=45, swing2LimitDegrees=45, twistLimitDegrees=45}
 {bone1Name=dyn_skirt_01_r, bone2Name=dyn_skirt_01_r_MSTR, swing1LimitDegrees=25, swing2LimitDegrees=45, twistLimitDegrees=5}
 {bone1Name=dyn_skirt_02_r, bone2Name=dyn_skirt_01_r, swing1LimitDegrees=25, swing2LimitDegrees=45, twistLimitDegrees=5}
 {bone1Name=dyn_skirt_03_r, bone2Name=dyn_skirt_02_r, swing1LimitDegrees=25, swing2LimitDegrees=45, twistLimitDegrees=5}
 {
  bone1Name=thigh_l
  bone2Name=pelvis
  swing1LimitDegrees=45
  swing1Motion=Free
  swing2LimitDegrees=45
  swing2Motion=Free
  twistLimitDegrees=45
  twistMotion=Free
 }
 {
  bone1Name=thigh_r
  bone2Name=pelvis
  swing1LimitDegrees=45
  swing1Motion=Free
  swing2LimitDegrees=45
  swing2Motion=Free
  twistLimitDegrees=45
  twistMotion=Free
 }
 {
  bone1Name=thigh_r
  bone2Name=dyn_skirt_01_r_MSTR
  swing1LimitDegrees=45
  swing1Motion=Free
  swing2LimitDegrees=45
  swing2Motion=Free
  twistLimitDegrees=45
  twistMotion=Free
 }
 {
  bone1Name=dyn_skirt_bk2_01_l
  bone2Name=dyn_skirt_bk_01_l
  swing1LimitDegrees=45
  swing1Motion=Free
  swing2LimitDegrees=45
  swing2Motion=Free
  twistLimitDegrees=45
  twistMotion=Free
 }
]
