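# Animation Blueprint for the F_MED_Cyber_Armor_Alt cosmetic body, adding secondary motion to skirt and collar bones.
# It copies the parent mesh pose, then chains RigidBody simulations, Anim Dynamics, constraints and corrective bone offsets.
# Update logic sets gravity override and state flags: backbling softens collar dynamics, crouching strengthens left skirt dynamics, and sliding or emoting pins bones to spine and thighs.
# Entering or leaving skydiving resets dynamics.
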AnimBlueprint /BRCosmetics/Characters/Player/Female/Medium/Bodies/F_MED_Cyber_Armor_Alt/Meshs/F_MED_Cyber_Armor_Alt_AnimBP
target_skeleton=&/F_MED_Cyber_Armor_Alt_Skeleton
graphs [
 AnimGraph {nodes [
   "Transform (Modify) Bone - Bone: dyn_skirt_fr_01_MSTR"!(ComponentPose, bAlphaBoolEnabled:bool=True, Translation="0,2,0", Rotation="0,0,0", Scale="1,1,1") -> (Pose) {
    alphaInputType=Bool
    rotationMode=BMM_Additive
    scaleMode=BMM_Additive
    translationMode=BMM_Additive
    alphaBoolBlend {blendInTime=0.1, blendOption=Linear, blendOutTime=0.1}
    boneToModify {boneName=dyn_skirt_fr_01_MSTR}
    translation {y=2}
   }
   n1 = SelectFloat(A=1, B=0.25, bPickA=Get Crouching() -> (Crouching:bool))
   n2 = RigidBody!(
    ComponentPose=Local To Component(LocalPose=Copy Pose From Mesh() {bCopyCurves=true, bUseAttachedParent=true})
    Alpha
    OverrideWorldGravity:Vector=Get GravityOverride()
    ExternalForce
   ) -> (Pose) {
    bClampLinearTranslationLimitToRefPose=true
    bOverrideWorldGravity=true
    overridePhysicsAsset=&/F_MED_Cyber_Armor_Alt_Dress_Physics
    simulationSpace=BaseBoneSpace
    alphaBoolBlend {blendOption=Linear}
    baseBoneRef {boneName=pelvis}
    componentAppliedLinearAccClamp=<3000 3000 850>
    componentLinearAccScale=<0.4 0.4 0.42>
    overrideWorldGravity {z=-980}
   }
   n3 = RigidBody(ComponentPose=@n2, OverrideWorldGravity="0,0,-980") {
    bClampLinearTranslationLimitToRefPose=true
    bOverrideWorldGravity=true
    overridePhysicsAsset=/BRCosmetics/Characters/Player/Female/Medium/Bodies/F_MED_Cyber_Karate_Human/Meshes/F_MED_Cyber_Karate_Human_Scarf_Physics
    simulationSpace=BaseBoneSpace
    alphaBoolBlend {blendOption=Linear}
    baseBoneRef {boneName=root}
    componentAppliedLinearAccClamp=<3000 3000 3000>
    componentLinearAccScale=<0.4 0.4 0.42>
    componentLinearVelScale=<0.4 0.4 0.4>
    overrideWorldGravity {z=-980}
   }
   n4 = "Constraint - dyn_skirt_bk2_01_r (R)"(ComponentPose=@n3, Alpha=0.25, ConstraintWeights_0=1) {
    alphaBoolBlend {blendOption=Linear}
    boneToModify {boneName=dyn_skirt_bk2_01_r}
    constraintSetup [{transformType=Rotation, targetBone {boneName=dyn_skirt_01_r}}]
    constraintWeights [1]
   }
   n5 = "Constraint - dyn_skirt_fr_01_MSTR (PP)"!(ComponentPose=@n4, Alpha=1, ConstraintWeights_0=37, ConstraintWeights_1:float=0.6) -> (Pose) {
    alphaBoolBlend {blendOption=Linear}
    boneToModify {boneName=dyn_skirt_fr_01_MSTR}
    constraintSetup [{targetBone {boneName=dyn_skirt_fr_01_l}}, {targetBone {boneName=dyn_skirt_fr_01_r}}]
    constraintWeights [1,1]
   }
   n6 = "Constraint - dyn_skirt_fr_01 (PP)"!(ComponentPose=@n5, Alpha=1, ConstraintWeights_0=0.4, ConstraintWeights_1:float=0.6) -> (Pose) {
    alphaBoolBlend {blendOption=Linear}
    boneToModify {boneName=dyn_skirt_fr_01}
    constraintSetup [{targetBone {boneName=dyn_skirt_fr_01_l}}, {targetBone {boneName=dyn_skirt_fr_01_r}}]
    constraintWeights [1,1]
   }
   n7 = "Constraint - dyn_skirt_fr_02 (PP)"!(ComponentPose=@n6, Alpha=1, ConstraintWeights_0=0.4, ConstraintWeights_1:float=0.6) -> (Pose) {
    alphaBoolBlend {blendOption=Linear}
    boneToModify {boneName=dyn_skirt_fr_02}
    constraintSetup [{targetBone {boneName=dyn_skirt_fr_01_l}}, {targetBone {boneName=dyn_skirt_fr_01_r}}]
    constraintWeights [1,1]
   }
   n8 = "Constraint - dyn_skirt_fr_03 (PP)"!(ComponentPose=@n7, Alpha=1, ConstraintWeights_0=0.4, ConstraintWeights_1:float=0.6) -> (Pose) {
    alphaBoolBlend {blendOption=Linear}
    boneToModify {boneName=dyn_skirt_fr_03}
    constraintSetup [{targetBone {boneName=dyn_skirt_fr_01_l}}, {targetBone {boneName=dyn_skirt_fr_01_r}}]
    constraintWeights [1,1]
   }
   n9 = SelectFloat(A=0.2, B=1, bPickA=Get wearingBackbling() -> (wearingBackbling:bool))
   n10 = Anim Dynamics - Bone: dyn_collar_back_l!(ComponentPose=@n8, Alpha=@n9, GravityOverride:Vector="0,0,0", ExternalForce="0,0,0") -> (Pose) {
    angularDampingOverride=0.9999
    angularSpringConstant=75
    bAngularSpring=true
    bGravityOverrideInSimSpace=true
    bOverrideAngularDamping=true
    bUseGravityOverride=true
    gravityScale=0.8
    simulationSpace=RootRelative
    alphaBoolBlend {blendOption=Linear}
    boundBone {boneName=dyn_collar_back_l}
    componentAppliedLinearAccClamp=<5000 5000 3000>
    componentLinearAccScale=<0.55 0.55 0.55>
    componentLinearVelScale=<0.55 0.55 0.55>
    physicsBodyDefinitions [{
      sphereCollisionRadius=0
      boundBone {boneName=dyn_collar_back_l}
      boxExtents=<20 20 10>
      constraintSetup {angularLimitsMax=<15 15 0>, angularLimitsMin=<-15 -15 -90>, angularTarget=<0 0 0>}
      localJointOffset=<2 0 0>
     }]
   }
   n11 = Anim Dynamics - Bone: dyn_collar_back_r!(ComponentPose=@n10, Alpha=@n9, GravityOverride:Vector="0,0,0", ExternalForce="0,0,0") -> (Pose) {
    angularDampingOverride=0.99
    angularSpringConstant=75
    bAngularSpring=true
    bGravityOverrideInSimSpace=true
    bOverrideAngularDamping=true
    bUseGravityOverride=true
    gravityScale=0.8
    simulationSpace=RootRelative
    alphaBoolBlend {blendOption=Linear}
    boundBone {boneName=dyn_collar_back_r}
    componentAppliedLinearAccClamp=<5000 5000 3000>
    componentLinearAccScale=<0.55 0.55 0.55>
    componentLinearVelScale=<0.55 0.55 0.55>
    physicsBodyDefinitions [{
      sphereCollisionRadius=0
      boundBone {boneName=dyn_collar_back_r}
      boxExtents=<20 20 10>
      constraintSetup {angularLimitsMax=<15 15 0>, angularLimitsMin=<-15 -15 -90>, angularTarget=<0 0 0>}
      localJointOffset=<2 0 0>
     }]
   }
   n12 = Get Skydiving() -> (Skydiving:bool)
   n13 = "Transform (Modify) Bone - Bone: master_collar_back_l"!(ComponentPose=@n11, bAlphaBoolEnabled:bool=@n12, Translation="0,0,2", Rotation="0.000,0.000,10", Scale="1,1,1") -> (Pose) {
    alphaInputType=Bool
    rotationMode=BMM_Additive
    scaleMode=BMM_Additive
    translationMode=BMM_Additive
    alphaBoolBlend {blendInTime=0.1, blendOption=Linear, blendOutTime=0.1}
    boneToModify {boneName=master_collar_back_l}
    rotation {roll=10}
    translation {z=2}
   }
   n14 = "Transform (Modify) Bone - Bone: master_collar_back_r"!(ComponentPose=@n13, bAlphaBoolEnabled:bool=@n12, Translation="0,0,2", Rotation="0,0,10", Scale="1,1,1") -> (Pose) {
    alphaInputType=Bool
    rotationMode=BMM_Additive
    scaleMode=BMM_Additive
    translationMode=BMM_Additive
    alphaBoolBlend {blendInTime=0.1, blendOption=Linear, blendOutTime=0.1}
    boneToModify {boneName=master_collar_back_r}
    rotation {roll=10}
    translation {z=2}
   }
   n15 = "Transform (Modify) Bone - Bone: master_collar_back_l"(ComponentPose=@n14, bAlphaBoolEnabled=True, Translation="0,0,2", Rotation="0.000,0.000,11", Scale="1,1,1") {
    alphaInputType=Bool
    rotationMode=BMM_Additive
    scaleMode=BMM_Additive
    translationMode=BMM_Additive
    alphaBoolBlend {blendInTime=0.1, blendOption=Linear, blendOutTime=0.1}
    boneToModify {boneName=master_collar_back_l}
    rotation {roll=11}
    translation {z=2}
   }
   n16 = "Transform (Modify) Bone - Bone: master_collar_back_r"(ComponentPose=@n15, bAlphaBoolEnabled=True, Translation="0,0,2", Rotation="0,0,11", Scale="1,1,1") {
    alphaInputType=Bool
    rotationMode=BMM_Additive
    scaleMode=BMM_Additive
    translationMode=BMM_Additive
    alphaBoolBlend {blendInTime=0.1, blendOption=Linear, blendOutTime=0.1}
    boneToModify {boneName=master_collar_back_r}
    rotation {roll=11}
    translation {z=2}
   }
   n17 = "Constraint - dyn_r_knee (PP)"!(ComponentPose=@n16, bAlphaBoolEnabled:bool=True, ConstraintWeights_0=0.2, ConstraintWeights_1:float=0.8) -> (Pose) {
    alphaInputType=Bool
    bAlphaBoolEnabled=false
    alphaBoolBlend {blendOption=Linear}
    boneToModify {boneName=dyn_r_knee}
    constraintSetup [{targetBone {boneName=thigh_twist_01_r}}, {targetBone {boneName=calf_r}}]
    constraintWeights [1,1]
   }
   n18 = "Constraint - dyn_l_knee (PP)"!(ComponentPose=@n17, bAlphaBoolEnabled:bool=True, ConstraintWeights_0=0.2, ConstraintWeights_1:float=0.8) -> (Pose) {
    alphaInputType=Bool
    bAlphaBoolEnabled=false
    alphaBoolBlend {blendOption=Linear}
    boneToModify {boneName=dyn_l_knee}
    constraintSetup [{targetBone {boneName=thigh_twist_01_l}}, {targetBone {boneName=calf_l}}]
    constraintWeights [1,1]
   }
   n19 = "Transform (Modify) Bone - Bone: dyn_l_knee"!(ComponentPose=@n18, bAlphaBoolEnabled:bool=True, Translation="0,-0.8,0", Rotation="0,0,0", Scale="1,1,1") -> (Pose) {
    alphaInputType=Bool
    bAlphaBoolEnabled=false
    translationMode=BMM_Additive
    translationSpace=BCS_BoneSpace
    alphaBoolBlend {blendOption=Linear}
    boneToModify {boneName=dyn_l_knee}
    translation {y=-0.8}
   }
   n20 = "Transform (Modify) Bone - Bone: dyn_r_knee"!(ComponentPose=@n19, bAlphaBoolEnabled:bool=True, Translation="-1,1,0", Rotation="0,0,0", Scale="1,1,1") -> (Pose) {
    alphaInputType=Bool
    bAlphaBoolEnabled=false
    translationMode=BMM_Additive
    translationSpace=BCS_BoneSpace
    alphaBoolBlend {blendOption=Linear}
    boneToModify {boneName=dyn_r_knee}
    translation=<-1 1>
   }
   n21 = "Transform (Modify) Bone - Bone: dyn_skirt_fr_01_l"!(ComponentPose=@n20, bAlphaBoolEnabled:bool=True, Translation="0,0,2", Rotation="0,0,0", Scale="1,1,1") -> (Pose) {
    alphaInputType=Bool
    bAlphaBoolEnabled=false
    rotationMode=BMM_Additive
    rotationSpace=BCS_BoneSpace
    scaleMode=BMM_Additive
    scaleSpace=BCS_BoneSpace
    translationMode=BMM_Additive
    translationSpace=BCS_BoneSpace
    alphaBoolBlend {blendInTime=0.1, blendOption=Linear, blendOutTime=0.1}
    boneToModify {boneName=dyn_skirt_fr_01_l}
    translation {z=2}
   }
   n22 = "Transform (Modify) Bone - Bone: dyn_skirt_fr_01_l"(ComponentPose=@n21, bAlphaBoolEnabled=True, Translation="13,0,0", Rotation="0,0,0", Scale="1,1,1") {
    alphaInputType=Bool
    bAlphaBoolEnabled=false
    rotationMode=BMM_Additive
    rotationSpace=BCS_BoneSpace
    scaleMode=BMM_Additive
    scaleSpace=BCS_BoneSpace
    translationMode=BMM_Additive
    translationSpace=BCS_BoneSpace
    alphaBoolBlend {blendInTime=0.1, blendOption=Linear, blendOutTime=0.1}
    boneToModify {boneName=dyn_skirt_fr_01_l}
    translation {x=13}
   }
   n23 = "Transform (Modify) Bone - Bone: dyn_skirt_01_r"!(ComponentPose=@n22, bAlphaBoolEnabled:bool=Get Jumping() -> (Jumping:bool), Translation="0,-2,5", Rotation="0,0,0", Scale="1,1,1") -> (Pose) {
    alphaInputType=Bool
    bAlphaBoolEnabled=false
    rotationMode=BMM_Additive
    rotationSpace=BCS_BoneSpace
    scaleMode=BMM_Additive
    scaleSpace=BCS_BoneSpace
    translationMode=BMM_Additive
    translationSpace=BCS_BoneSpace
    alphaBoolBlend {blendInTime=0.1, blendOption=Linear, blendOutTime=0.1}
    boneToModify {boneName=dyn_skirt_01_r}
    translation {y=-2, z=5}
   }
   n24 = "Transform (Modify) Bone - Bone: dyn_skirt_fr_01_MSTR"(ComponentPose=@n23, bAlphaBoolEnabled=True, Translation="2,0,5", Rotation="0,-15,0", Scale="1,1,1") {
    alphaInputType=Bool
    bAlphaBoolEnabled=false
    rotationMode=BMM_Additive
    rotationSpace=BCS_BoneSpace
    scaleMode=BMM_Additive
    scaleSpace=BCS_BoneSpace
    translationMode=BMM_Additive
    translationSpace=BCS_BoneSpace
    alphaBoolBlend {blendInTime=0.1, blendOption=Linear, blendOutTime=0.1}
    boneToModify {boneName=dyn_skirt_fr_01_MSTR}
    rotation {yaw=-15}
    translation {x=2, z=5}
   }
   n25 = Anim Dynamics - Bone: dyn_skirt_fr_01_l!(ComponentPose=@n24, Alpha=@n1, GravityOverride:Vector="0,0,0", ExternalForce="0,0,0") -> (Pose) {
    angularDampingOverride=0.9999
    angularSpringConstant=75
    bAngularSpring=true
    bGravityOverrideInSimSpace=true
    bOverrideAngularDamping=true
    bUseGravityOverride=true
    simulationSpace=RootRelative
    alphaBoolBlend {blendOption=Linear}
    boundBone {boneName=dyn_skirt_fr_01_l}
    componentLinearAccScale=<1 1 1>
    componentLinearVelScale=<1 1 1>
    physicsBodyDefinitions [{
      sphereCollisionRadius=0
      boundBone {boneName=dyn_skirt_fr_01_l}
      boxExtents=<20 20 10>
      constraintSetup {angularLimitsMax=<25 5 0>, angularLimitsMin=<-25 -25 0>, angularTarget=<0 0 0>}
      localJointOffset=<2 0 0>
     }]
   }
   n26 = Anim Dynamics - Bone: dyn_skirt_fr_02_l!(ComponentPose=@n25, Alpha=@n1, GravityOverride:Vector="0,0,0", ExternalForce="0,0,0") -> (Pose) {
    angularDampingOverride=0.9999
    angularSpringConstant=75
    bAngularSpring=true
    bGravityOverrideInSimSpace=true
    bOverrideAngularDamping=true
    bUseGravityOverride=true
    simulationSpace=RootRelative
    alphaBoolBlend {blendOption=Linear}
    boundBone {boneName=dyn_skirt_fr_02_l}
    componentLinearAccScale=<1 1 1>
    componentLinearVelScale=<1 1 1>
    physicsBodyDefinitions [{
      sphereCollisionRadius=0
      boundBone {boneName=dyn_skirt_fr_02_l}
      boxExtents=<20 20 10>
      constraintSetup {angularLimitsMax=<25 17 0>, angularLimitsMin=<-25 -25 0>, angularTarget=<0 0 0>}
      localJointOffset=<2 0 0>
     }]
   }
   n27 = Anim Dynamics - Bone: dyn_skirt_fr_03_l!(ComponentPose=@n26, Alpha=@n1, GravityOverride:Vector="0,0,0", ExternalForce="0,0,0") -> (Pose) {
    angularDampingOverride=0.9999
    angularSpringConstant=75
    bAngularSpring=true
    bGravityOverrideInSimSpace=true
    bOverrideAngularDamping=true
    bUseGravityOverride=true
    simulationSpace=RootRelative
    alphaBoolBlend {blendOption=Linear}
    boundBone {boneName=dyn_skirt_fr_03_l}
    componentLinearAccScale=<1 1 1>
    componentLinearVelScale=<1 1 1>
    physicsBodyDefinitions [{
      sphereCollisionRadius=0
      boundBone {boneName=dyn_skirt_fr_03_l}
      boxExtents=<20 20 10>
      constraintSetup {angularLimitsMax=<25 17 0>, angularLimitsMin=<-25 -25 0>, angularTarget=<0 0 0>}
      localJointOffset=<2 0 0>
     }]
   }
   n28 = Anim Dynamics - Bone: dyn_skirt_fr_01_r!(ComponentPose=@n27, bAlphaBoolEnabled:bool=True, GravityOverride:Vector="0,0,0", ExternalForce="0,0,0") -> (Pose) {
    alphaInputType=Bool
    angularDampingOverride=0.9999
    angularSpringConstant=75
    bAngularSpring=true
    bGravityOverrideInSimSpace=true
    bOverrideAngularDamping=true
    bUseGravityOverride=true
    simulationSpace=RootRelative
    alphaBoolBlend {blendOption=Linear}
    boundBone {boneName=dyn_skirt_fr_01_r}
    componentAppliedLinearAccClamp=<5000 5000 5000>
    componentLinearAccScale=<0.5 0.5 0.5>
    componentLinearVelScale=<0.5 0.5 0.5>
    physicsBodyDefinitions [{
      sphereCollisionRadius=0
      boundBone {boneName=dyn_skirt_fr_01_r}
      boxExtents=<20 20 10>
      constraintSetup {angularLimitsMax=<25 -5 0>, angularLimitsMin=<-25 -25 0>, angularTarget=<0 0 0>}
      localJointOffset=<2 0 0>
     }]
   }
   n29 = Anim Dynamics - Bone: dyn_skirt_fr_02_r!(ComponentPose=@n28, bAlphaBoolEnabled:bool=True, GravityOverride:Vector="0,0,0", ExternalForce="0,0,0") -> (Pose) {
    alphaInputType=Bool
    angularDampingOverride=0.9999
    angularSpringConstant=75
    bAngularSpring=true
    bGravityOverrideInSimSpace=true
    bOverrideAngularDamping=true
    bUseGravityOverride=true
    simulationSpace=RootRelative
    alphaBoolBlend {blendOption=Linear}
    boundBone {boneName=dyn_skirt_fr_02_r}
    componentAppliedLinearAccClamp=<5000 5000 5000>
    componentLinearAccScale=<0.5 0.5 0.5>
    componentLinearVelScale=<0.5 0.5 0.5>
    physicsBodyDefinitions [{
      sphereCollisionRadius=0
      boundBone {boneName=dyn_skirt_fr_02_r}
      boxExtents=<20 20 10>
      constraintSetup {angularLimitsMax=<25 -5 0>, angularLimitsMin=<-25 -25 0>, angularTarget=<0 0 0>}
      localJointOffset=<2 0 0>
     }]
   }
   n30 = Anim Dynamics - Bone: dyn_skirt_fr_03!(ComponentPose=@n29, bAlphaBoolEnabled:bool=True, GravityOverride:Vector="0,0,0", ExternalForce="0,0,0") -> (Pose) {
    alphaInputType=Bool
    angularDampingOverride=0.9999
    angularSpringConstant=75
    bAngularSpring=true
    bGravityOverrideInSimSpace=true
    bOverrideAngularDamping=true
    bUseGravityOverride=true
    simulationSpace=RootRelative
    alphaBoolBlend {blendOption=Linear}
    boundBone {boneName=dyn_skirt_fr_03}
    componentAppliedLinearAccClamp=<5000 5000 5000>
    componentLinearAccScale=<0.5 0.5 0.5>
    componentLinearVelScale=<0.5 0.5 0.5>
    physicsBodyDefinitions [{
      sphereCollisionRadius=0
      boundBone {boneName=dyn_skirt_fr_03}
      boxExtents=<20 20 10>
      constraintSetup {angularLimitsMax=<25 -5 0>, angularLimitsMin=<-25 -25 0>, angularTarget=<0 0 0>}
      localJointOffset=<2 0 0>
     }]
   }
   n31 = Get AllSliding() -> (AllSliding:bool)
   n32 = "Constraint - dyn_collar_back_l (P)"!(ComponentPose=@n30, bAlphaBoolEnabled:bool=@n31, ConstraintWeights_0=1) -> (Pose) {
    alphaInputType=Bool
    alphaBoolBlend {blendInTime=0.1, blendOption=Linear, blendOutTime=0.1}
    boneToModify {boneName=dyn_collar_back_l}
    constraintSetup [{targetBone {boneName=spine_05}}]
    constraintWeights [1]
   }
   n33 = "Constraint - dyn_collar_back_r (P)"!(ComponentPose=@n32, bAlphaBoolEnabled:bool=@n31, ConstraintWeights_0=1) -> (Pose) {
    alphaInputType=Bool
    alphaBoolBlend {blendInTime=0.1, blendOption=Linear, blendOutTime=0.1}
    boneToModify {boneName=dyn_collar_back_r}
    constraintSetup [{targetBone {boneName=spine_05}}]
    constraintWeights [1]
   }
   n34 = "Constraint - dyn_skirt_01_r (P)"!(ComponentPose=@n33, bAlphaBoolEnabled:bool=@n31, ConstraintWeights_0=1) -> (Pose) {
    alphaInputType=Bool
    alphaBoolBlend {blendInTime=0.1, blendOption=Linear, blendOutTime=0.1}
    boneToModify {boneName=dyn_skirt_01_r}
    constraintSetup [{targetBone {boneName=thigh_r}}]
    constraintWeights [1]
   }
   n35 = "Constraint - dyn_skirt_02_r (P)"!(ComponentPose=@n34, bAlphaBoolEnabled:bool=@n31, ConstraintWeights_0=1) -> (Pose) {
    alphaInputType=Bool
    alphaBoolBlend {blendInTime=0.1, blendOption=Linear, blendOutTime=0.1}
    boneToModify {boneName=dyn_skirt_02_r}
    constraintSetup [{targetBone {boneName=thigh_r}}]
    constraintWeights [1]
   }
   n36 = "Constraint - dyn_skirt_03_r (P)"!(ComponentPose=@n35, bAlphaBoolEnabled:bool=@n31, ConstraintWeights_0=1) -> (Pose) {
    alphaInputType=Bool
    alphaBoolBlend {blendInTime=0.1, blendOption=Linear, blendOutTime=0.1}
    boneToModify {boneName=dyn_skirt_03_r}
    constraintSetup [{targetBone {boneName=thigh_r}}]
    constraintWeights [1]
   }
   n37 = "Constraint - dyn_skirt_fr_01_l (P)"!(ComponentPose=@n36, bAlphaBoolEnabled:bool=@n31, ConstraintWeights_0=1) -> (Pose) {
    alphaInputType=Bool
    alphaBoolBlend {blendInTime=0.1, blendOption=Linear, blendOutTime=0.1}
    boneToModify {boneName=dyn_skirt_fr_01_l}
    constraintSetup [{targetBone {boneName=thigh_l}}]
    constraintWeights [1]
   }
   n38 = "Constraint - dyn_skirt_fr_02_l (P)"!(ComponentPose=@n37, bAlphaBoolEnabled:bool=@n31, ConstraintWeights_0=1) -> (Pose) {
    alphaInputType=Bool
    alphaBoolBlend {blendInTime=0.1, blendOption=Linear, blendOutTime=0.1}
    boneToModify {boneName=dyn_skirt_fr_02_l}
    constraintSetup [{targetBone {boneName=thigh_l}}]
    constraintWeights [1]
   }
   n39 = "Constraint - dyn_skirt_fr_03_l (P)"!(ComponentPose=@n38, bAlphaBoolEnabled:bool=@n31, ConstraintWeights_0=1) -> (Pose) {
    alphaInputType=Bool
    alphaBoolBlend {blendInTime=0.1, blendOption=Linear, blendOutTime=0.1}
    boneToModify {boneName=dyn_skirt_fr_03_l}
    constraintSetup [{targetBone {boneName=thigh_l}}]
    constraintWeights [1]
   }
   n40 = "Constraint - dyn_skirt_fr_01_r (P)"!(ComponentPose=@n39, bAlphaBoolEnabled:bool=@n31, ConstraintWeights_0=1) -> (Pose) {
    alphaInputType=Bool
    alphaBoolBlend {blendInTime=0.1, blendOption=Linear, blendOutTime=0.1}
    boneToModify {boneName=dyn_skirt_fr_01_r}
    constraintSetup [{targetBone {boneName=thigh_r}}]
    constraintWeights [1]
   }
   n41 = "Constraint - dyn_skirt_fr_02_r (P)"!(ComponentPose=@n40, bAlphaBoolEnabled:bool=@n31, ConstraintWeights_0=1) -> (Pose) {
    alphaInputType=Bool
    alphaBoolBlend {blendInTime=0.1, blendOption=Linear, blendOutTime=0.1}
    boneToModify {boneName=dyn_skirt_fr_02_r}
    constraintSetup [{targetBone {boneName=thigh_r}}]
    constraintWeights [1]
   }
   n42 = "Constraint - dyn_skirt_fr_03_r (P)"!(ComponentPose=@n41, bAlphaBoolEnabled:bool=@n31, ConstraintWeights_0=1) -> (Pose) {
    alphaInputType=Bool
    alphaBoolBlend {blendInTime=0.1, blendOption=Linear, blendOutTime=0.1}
    boneToModify {boneName=dyn_skirt_fr_03_r}
    constraintSetup [{targetBone {boneName=thigh_r}}]
    constraintWeights [1]
   }
   n43 = "Constraint - dyn_skirt_fr_01 (PP)"!(ComponentPose=@n42, bAlphaBoolEnabled:bool=@n31, ConstraintWeights_0=0, ConstraintWeights_1=1) -> (Pose) {
    alphaInputType=Bool
    alphaBoolBlend {blendInTime=0.1, blendOption=Linear, blendOutTime=0.1}
    boneToModify {boneName=dyn_skirt_fr_01}
    constraintSetup [{targetBone {boneName=thigh_twist_01_l}}, {targetBone {boneName=thigh_twist_01_r}}]
    constraintWeights [1,1]
   }
   n44 = "Constraint - dyn_skirt_fr_02 (PP)"!(ComponentPose=@n43, bAlphaBoolEnabled:bool=@n31, ConstraintWeights_0=0, ConstraintWeights_1=1) -> (Pose) {
    alphaInputType=Bool
    alphaBoolBlend {blendInTime=0.1, blendOption=Linear, blendOutTime=0.1}
    boneToModify {boneName=dyn_skirt_fr_02}
    constraintSetup [{targetBone {boneName=thigh_twist_01_l}}, {targetBone {boneName=thigh_twist_01_r}}]
    constraintWeights [1,1]
   }
   n45 = "Constraint - dyn_skirt_fr_03 (PP)"!(ComponentPose=@n44, bAlphaBoolEnabled:bool=@n31, ConstraintWeights_0=0, ConstraintWeights_1=1) -> (Pose) {
    alphaInputType=Bool
    alphaBoolBlend {blendInTime=0.1, blendOption=Linear, blendOutTime=0.1}
    boneToModify {boneName=dyn_skirt_fr_03}
    constraintSetup [{targetBone {boneName=thigh_twist_01_l}}, {targetBone {boneName=thigh_twist_01_r}}]
    constraintWeights [1,1]
   }
   n46 = Get AllSliding()
   n47 = "Transform (Modify) Bone - Bone: dyn_skirt_fr_01_MSTR"(ComponentPose=@n45, bAlphaBoolEnabled=@n46, Translation="0,0,6", Rotation="10,-8,0", Scale="1,1,1") {
    alphaInputType=Bool
    rotationMode=BMM_Additive
    rotationSpace=BCS_BoneSpace
    scaleMode=BMM_Additive
    scaleSpace=BCS_BoneSpace
    translationMode=BMM_Additive
    translationSpace=BCS_BoneSpace
    alphaBoolBlend {blendInTime=0.1, blendOption=Linear, blendOutTime=0.1}
    boneToModify {boneName=dyn_skirt_fr_01_MSTR}
    rotation {pitch=10, yaw=-8}
    translation {z=6}
   }
   n48 = "Transform (Modify) Bone - Bone: dyn_skirt_fr_01_l_MSTR"!(ComponentPose=@n47, bAlphaBoolEnabled:bool=@n46, Translation="0,0,1", Rotation="0,0,0", Scale="1,1,1") -> (Pose) {
    alphaInputType=Bool
    rotationMode=BMM_Additive
    rotationSpace=BCS_BoneSpace
    scaleMode=BMM_Additive
    scaleSpace=BCS_BoneSpace
    translationMode=BMM_Additive
    translationSpace=BCS_BoneSpace
    alphaBoolBlend {blendInTime=0.1, blendOption=Linear, blendOutTime=0.1}
    boneToModify {boneName=dyn_skirt_fr_01_l_MSTR}
    translation {z=1}
   }
   Output Pose(Result=Component To Local(ComponentPose=@n48)) {layerGroup=None}
  ]}
 EventGraph {nodes [
   Event BlueprintUpdateAnimation()
   > n1 = Sequence!(execute) -> (then_0, then_1, then_2:Exec)
   n2 = CheckForBackbling()
   > Set wearingBackbling(wearingBackbling:bool=@n2) -> (Output_Get:bool)
   n3 = Gravity Override State Function DEPRECATED(
    execute=@n1.then_1
    CharacterPartAnimInstance=Self-Reference()
    GameTimeInSeconds=GetGameTimeInSeconds()
    GravityOverrideParams=Get Gravity Override Params()
   )
   > Set GravityOverride(GravityOverride=@n3)
   Set Skydiving(execute=@n1.then_2, Skydiving:bool=OR Boolean(A=Get bIsSkydiveFloating(), B=Get bIsSkyDiving())) -> (Output_Get:bool)
   > Set Jumping(Jumping:bool=OR Boolean(A=Get bIsJumping(), B=Get bIsFalling())) -> (Output_Get:bool)
   > Set Crouching(Crouching:bool=Get bIsCrouching()) -> (Output_Get:bool)
   n4 = Get bIsJumping()
   n5 = NOT Boolean(A=Get bIsFalling())
   n6 = OR Boolean!(A=Get bIsSliding(), B=Get bIsSlopeSliding() -> (bIsSlopeSliding:bool), C:bool=Get bIsPlayingEmote()) -> (ReturnValue)
   > Set AllSliding(AllSliding:bool=AND Boolean(A=AND Boolean(A=@n6, B=@n5), B=NOT Boolean(A=@n4))) -> (Output_Get:bool)
   n7 = Event OnBeginSkydiving()
   ResetDynamics(execute=[@n7.then, @n8.then], InTeleportType=TeleportPhysics)
   > ResetDynamics(InTeleportType=ResetPhysics)
   n8 = Event OnEndSkydiving()
   TryGetPawnOwner()
   n9 = "To Float (Boolean)"(InBool=@n2)
   float - float(B=@n9) {is=title K2Node_CallFunction}
   Set backblingAlpha(backblingAlpha:double=@n9) -> (Output_Get:double)
  ]}
]
variables [
 AllSliding {type=boolean}
 backblingAlpha {type=number}
 Crouching {type=boolean}
 Gravity Override Params {
  type=GravityOverrideParamsStruct
  default {
   skydive_DivingDown {gravityOverride=<-1500 0 200>}
   skydive_FloatingDown {gravityOverride=<0 0 600>}
   skydive_GoingUp {gravityOverride=<0 0 -1000>}
   swimming {gravityOverride=<0 0 0>}
  }
 }
 GravityOverride {type=Vector}
 Jumping {type=boolean}
 Skydiving {type=boolean}
 wearingBackbling {type=boolean}
]
