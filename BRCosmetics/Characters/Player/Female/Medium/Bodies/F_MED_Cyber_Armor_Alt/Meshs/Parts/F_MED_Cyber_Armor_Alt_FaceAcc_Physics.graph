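# Physics asset for the FaceAcc part of the female medium Cyber Armor Alt character body, defining physics bodies for its dynamic hair bones.
# Three capsule chains (a, b, g) hang from a head body built from four overlapping capsules, with b forking side branches at its second and third bones.
# Constraints set narrow swing and tighter twist limits, with the b chain's root held stiffest against the head.
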PhysicsAsset /BRCosmetics/Characters/Player/Female/Medium/Bodies/F_MED_Cyber_Armor_Alt/Meshs/Parts/F_MED_Cyber_Armor_Alt_FaceAcc_Physics
bodies {
 dyn_hair_a_01 [{shapeType=Capsule, length=2.7932, radius=0.8, shapeName=dyn_hair_a_01_capsule, center=<2.1125 0 0>, rotation=<rot 90 0 0>}]
 dyn_hair_a_02 [{shapeType=Capsule, length=2.4134, radius=0.8, shapeName=dyn_hair_a_02_capsule, center=<2.0272 0 0>, rotation=<rot 90 0 0>}]
 dyn_hair_a_03 [{shapeType=Capsule, length=3.1834, radius=0.8, shapeName=dyn_hair_a_03_capsule, center=<2.3175 0 0>, rotation=<rot 90 0 0>}]
 dyn_hair_b_01 [{shapeType=Capsule, length=3.413, radius=0.8, shapeName=dyn_hair_b_01_capsule, center=<2.5709 0 0>, rotation=<rot 90 0 0>}]
 dyn_hair_b_02 [{shapeType=Capsule, length=2.6619, radius=0.8, shapeName=dyn_hair_b_02_capsule, center=<2.0163 0 0>, rotation=<rot 90 0 0>}]
 dyn_hair_b_03 [{shapeType=Capsule, length=3.4518, radius=0.8, shapeName=dyn_hair_b_03_capsule, center=<2.4423 0 0>, rotation=<rot 90 0 0>}]
 dyn_hair_b_04 [{shapeType=Capsule, length=3.1035, radius=0.8, shapeName=dyn_hair_b_04_capsule, center=<2.2625 0 0>, rotation=<rot 90 0 0>}]
 dyn_hair_b_05 [{shapeType=Capsule, length=2.8994, radius=0.8, shapeName=dyn_hair_b_05_capsule, center=<2.044 -0.0001 -0.6233>, rotation=<rot 70.6272 0 0>}]
 dyn_hair_b_06 [{shapeType=Capsule, length=1.3341, radius=0.8, shapeName=dyn_hair_b_06_capsule, center=<1.4482 0 0>, rotation=<rot 90 0 0>}]
 dyn_hair_b_07 [{shapeType=Capsule, length=1.8786, radius=0.8, shapeName=dyn_hair_b_07_capsule, center=<1.7135 0 0>, rotation=<rot 90 0 0>}]
 dyn_hair_g_01 [{shapeType=Capsule, length=3.264, radius=0.8, shapeName=dyn_hair_g_01_capsule, center=<2.4106 0 0>, rotation=<rot 90 0 0>}]
 dyn_hair_g_02 [{shapeType=Capsule, length=2.6285, radius=0.8, shapeName=dyn_hair_g_02_capsule, center=<1.9453 0 0>, rotation=<rot 90 0 0>}]
 dyn_hair_g_03 [{shapeType=Capsule, length=2.4785, radius=0.8, shapeName=dyn_hair_g_03_capsule, center=<2.0526 0 0>, rotation=<rot 90 0 0>}]
 dyn_hair_g_04 [{shapeType=Capsule, length=3.3743, radius=0.8, shapeName=dyn_hair_g_04_capsule, center=<2.4038 0.0001 0>, rotation=<rot 90 0 0>}]
 head [
  {shapeType=Capsule, length=2.9868, radius=6.1877, shapeName=head_capsule, center=<3.8892 3.4076 0.3462>, rotation=<rot 90 90 78.1872>}
  {shapeType=Capsule, length=1.9671, radius=4.3636, shapeName=head_capsule, center=<-0.6706 4.2502 -0.0044>, rotation=<rot 63.3571 -37.2104 0>}
  {
   shapeType=Capsule
   length=2.1935
   radius=3.9443
   shapeName=head_capsule
   center=<-0.8263 4.4634 1.1656>
   rotation=<rot 58.5098 142.7918 179.9999>
  }
  {shapeType=Capsule, length=2.0279, radius=7.1467, shapeName=head_capsule, center=<5.4616 -1.9868 0.109>, rotation=<rot 90 -24.6235 -36.4363>}
 ]
}
constraints [
 {bone1Name=dyn_hair_b_03, bone2Name=dyn_hair_b_02, swing1LimitDegrees=20, swing2LimitDegrees=20, twistLimitDegrees=5}
 {bone1Name=dyn_hair_b_04, bone2Name=dyn_hair_b_03, swing1LimitDegrees=20, swing2LimitDegrees=20, twistLimitDegrees=5}
 {bone1Name=dyn_hair_b_05, bone2Name=dyn_hair_b_04, swing1LimitDegrees=20, swing2LimitDegrees=20, twistLimitDegrees=5}
 {bone1Name=dyn_hair_b_07, bone2Name=dyn_hair_b_03, swing1LimitDegrees=20, swing2LimitDegrees=20, twistLimitDegrees=5}
 {bone1Name=dyn_hair_b_06, bone2Name=dyn_hair_b_02, swing1LimitDegrees=20, swing2LimitDegrees=20, twistLimitDegrees=5}
 {bone1Name=dyn_hair_b_02, bone2Name=dyn_hair_b_01, swing1LimitDegrees=20, swing2LimitDegrees=20, twistLimitDegrees=5}
 {bone1Name=dyn_hair_b_01, bone2Name=head, swing1LimitDegrees=5, swing2LimitDegrees=5, twistLimitDegrees=5}
 {bone1Name=dyn_hair_g_02, bone2Name=dyn_hair_g_01, swing1LimitDegrees=20, swing2LimitDegrees=20, twistLimitDegrees=5}
 {bone1Name=dyn_hair_g_03, bone2Name=dyn_hair_g_02, swing1LimitDegrees=20, swing2LimitDegrees=20, twistLimitDegrees=5}
 {bone1Name=dyn_hair_a_01, bone2Name=head, swing1LimitDegrees=20, swing2LimitDegrees=20, twistLimitDegrees=5}
 {bone1Name=dyn_hair_g_01, bone2Name=head, swing1LimitDegrees=20, swing2LimitDegrees=20, twistLimitDegrees=5}
 {bone1Name=dyn_hair_g_04, bone2Name=dyn_hair_g_03, swing1LimitDegrees=20, swing2LimitDegrees=20, twistLimitDegrees=5}
 {bone1Name=dyn_hair_a_02, bone2Name=dyn_hair_a_01, swing1LimitDegrees=20, swing2LimitDegrees=20, twistLimitDegrees=5}
 {bone1Name=dyn_hair_a_03, bone2Name=dyn_hair_a_02, swing1LimitDegrees=20, swing2LimitDegrees=20, twistLimitDegrees=5}
]
