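# Physics asset for the face accessory mesh of the female medium Cyber Armor Alt body, covering a side hair bun chain.
# Capsule bodies sit on the head and on two dynamic bun bones, dyn_bunSide and dyn_bunSide1.
# Constraints link head to dyn_bunSide to dyn_bunSide1 in a chain, each with equal swing and twist angle limits.
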PhysicsAsset /BRCosmetics/Characters/Player/Female/Medium/Bodies/F_MED_Cyber_Armor_Alt/Meshs/Parts/F_MED_Cyber_Armor_Alt_FaceAcc_Physics1
bodies {
 dyn_bunSide [{shapeType=Capsule, length=2.1217, radius=0.8783, shapeName=dyn_bunSide_capsule, center=<2.9852 0 0>, rotation=<rot 90 0 0>}]
 dyn_bunSide1 [{
   shapeType=Capsule
   length=4.2541
   radius=0.9171
   shapeName=dyn_bunSide1_capsule
   center=<3.2486 0 -0.2749>
   rotation=<rot 86.1288 -0.0003 -0.0003>
  }]
 head [{shapeType=Capsule, length=1.45, radius=9.14, center=<5.7352 -0.7905 1.5392>}]
}
constraints [
 {bone1Name=dyn_bunSide, bone2Name=head, swing1LimitDegrees=45, swing2LimitDegrees=45, twistLimitDegrees=45}
 {bone1Name=dyn_bunSide1, bone2Name=dyn_bunSide, swing1LimitDegrees=45, swing2LimitDegrees=45, twistLimitDegrees=45}
]
